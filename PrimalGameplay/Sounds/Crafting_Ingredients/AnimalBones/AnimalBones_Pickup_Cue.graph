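# AnimalBones_Pickup_Cue is a SoundCue for picking up the animal bones crafting ingredient.
# A Random node picks between two equally weighted pickup waves and feeds a Modulator that varies pitch slightly while holding volume constant.
# The cue plays at reduced volume on the loot-collect sound class, with weapon-drop attenuation settings for spatial falloff.
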SoundCue /PrimalGameplay/Sounds/Crafting_Ingredients/AnimalBones/AnimalBones_Pickup_Cue
root=Modulator_0
nodes [
 {id=Modulator_0, children [Random_0], properties {pitchMin=0.97, volumeMax=1, volumeMin=1}}
 {id=Random_0, children [WavePlayer_1,WavePlayer_2], properties {weights [1,1]}}
 {id=WavePlayer_1, properties {soundWaveAssetPtr=&/AnimalBones_Pickup_01}}
 {id=WavePlayer_2, properties {soundWaveAssetPtr=&/AnimalBones_Pickup_02}}
]
properties {
 attenuationSettings=&Fort_Weapon_Drop_Att
 maxDistance=3000
 soundClassObject=&Fort_Loot_Collect
 volumeMultiplier=0.25
}
