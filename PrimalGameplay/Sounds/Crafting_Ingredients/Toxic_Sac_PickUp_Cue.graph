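# Sound cue for picking up a Toxic Sac crafting ingredient, in the loot-collect sound class with weapon-drop attenuation.
# A player melee equip sound plays at once, mixed with a second layer that starts after a short fixed delay.
# That delayed layer mixes one of three equally weighted Toxic Sac pickup variants with a quiet generic loot special-effect sound, which a volume envelope fades out over about a second.
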SoundCue /PrimalGameplay/Sounds/Crafting_Ingredients/Toxic_Sac_PickUp_Cue
root=Mixer_0
nodes [
 {id=Delay_2, children [Mixer_2], properties {delayMax=0.1, delayMin=0.1}}
 {
  id=Enveloper_0
  children [WavePlayer_4]
  properties {
   pitchCurve {editorCurveData {keys [{time=0}]}}
   volumeCurve {editorCurveData {keys [<at 0 1>,<at 0.2365 0.9984>,<at 0.5256 0.4455>,<at 0.9477 -0.0016>]}}
  }
 }
 {id=Mixer_0, children [Delay_2,WavePlayer_1], properties {inputVolume [1.2,0.5]}}
 {id=Mixer_2, children [Enveloper_0,Random_0], properties {inputVolume [0.15,1]}}
 {id=Random_0, children [WavePlayer_0,WavePlayer_5,WavePlayer_6], properties {weights [1,1,1]}}
 {id=WavePlayer_0, properties {soundWaveAssetPtr=&/Toxic_Sac_PickUp_01}}
 {id=WavePlayer_1, properties {soundWaveAssetPtr=/Game/Sounds/Fort_GamePlay_Sounds/Player/player_melee_equip_02}}
 {id=WavePlayer_4, properties {soundWaveAssetPtr=/Game/Sounds/Foley_Loot/Ingredients/Generic/find_loot_special_effect}}
 {id=WavePlayer_5, properties {soundWaveAssetPtr=&/Toxic_Sac_PickUp_02}}
 {id=WavePlayer_6, properties {soundWaveAssetPtr=&/Toxic_Sac_PickUp_03}}
]
properties {
 attenuationSettings=&Fort_Weapon_Drop_Att
 maxDistance=3000
 soundClassObject=&Fort_Loot_Collect
 volumeMultiplier=0.5
}
